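# Material function in the Cactus Dancer Female pickaxe's Equalizer material folder that blends two colors without the washed-out midpoint of a plain lerp.
# Inputs a, b and scalar alpha feed one custom HLSL node, which lerps them and measures how far saturation drifts from the interpolated endpoint saturations.
# It pushes the color away from gray, scaled by that drift, lightness and a fixed strength, then outputs the saturated (clamped) result.
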MaterialFunction /BRCosmetics/Weapons/FORT_Melee/Pickaxe_Cactus_Dancer_Female/Materials/Equalizer/ColorLerpNonWashedOut
graph {nodes [
  MaterialExpressionFunctionInput_0 = FunctionInput() {inputName=a}
  MaterialExpressionFunctionInput_1 = FunctionInput() {inputName=b, sortPriority=1}
  MaterialExpressionFunctionInput_2 = FunctionInput() {inputName=alpha, inputType=FunctionInput_Scalar, sortPriority=2}
  n1 = Custom!(a=@MaterialExpressionFunctionInput_0, b=@MaterialExpressionFunctionInput_1, x=@MaterialExpressionFunctionInput_2) {
   code="struct Functions\r\n{\r\n    float getsat(float3 c)\r\n    {\r\n        float mi = min(min(c.x, c.y), c.z);\r\n        float ma = max(max(c.x, c.y), c.z);\r\n        return (ma - mi)/(ma+ 1e-7);\r\n    }\r\n};\r\nFunctions f;\r\n\r\n\r\n##define DSP_STR  3\r\n\r\n\r\n//Interpolated base color (with singularity fix)\r\nfloat3 ic = lerp(a, b, x) + float3(1e-6,0.,0.);\r\n\r\n//Saturation difference from ideal scenario\r\nfloat sd = abs(f.getsat(ic) - lerp(f.getsat(a), f.getsat(b), x));\r\n\r\n//Displacement direction\r\nfloat3 dir = normalize(float3(2.*ic.x - ic.y - ic.z, 2.*ic.y - ic.x - ic.z, 2.*ic.z - ic.y - ic.x));\r\n//Simple Lighntess\r\nfloat lgt = dot(float3(1.0, 1.0, 1.0), ic);\r\n\r\n//Extra scaling factor for the displacement\r\nfloat ff = dot(dir, normalize(ic));\r\n\r\n//Displace the color\r\nic += DSP_STR*dir*sd*ff*lgt;\r\nreturn saturate(ic);\r\n"
   description=ColorLerp
   inputs [
    {
     inputName=a
     input {expression=&self:MaterialExpressionFunctionInput_0}
    }
    {
     inputName=b
     input {expression=&self:MaterialExpressionFunctionInput_1}
    }
    {
     inputName=x
     input {expression=&self:MaterialExpressionFunctionInput_2}
    }
   ]
  }
  FunctionOutput(None=@n1)
 ]}
