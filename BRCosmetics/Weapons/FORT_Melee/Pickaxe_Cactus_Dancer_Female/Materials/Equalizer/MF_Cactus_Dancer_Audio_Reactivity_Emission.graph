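# Material function that adds an audio-reactive equalizer glow to the Cactus Dancer pickaxe's material attributes.
# UV1 forms columns that sample music history through SampleAudioMPCHistory, with custom HLSL selecting band or amplitude and gradient color by vertex green.
# Level thresholds light the bars, with layered sines adding idle motion as music quiets.
# Vertex red and blue drive gradient and mask; the gain-boosted result blends into emissive and base color.
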MaterialFunction /BRCosmetics/Weapons/FORT_Melee/Pickaxe_Cactus_Dancer_Female/Materials/Equalizer/MF_Cactus_Dancer_Audio_Reactivity_Emission
graph {nodes [
  n1 = TextureCoordinate()
  n2 = Add(A=LinearInterpolate(
    A=ScalarParameter() {defaultValue=0.6667, parameterName=bluehue}
    B=ScalarParameter() {defaultValue=0.8333, parameterName=phue}
    Alpha=ComponentMask(None=@n1) {g=true}
   )) {constB=0}
  n3 = Constant() {r=1}
  n4 = Constant() {r=1}
  Add(A=MaterialFunctionCall!(HSV_In=MaterialFunctionCall(X=@n2, Y=@n3, Z=@n4) {materialFunction=&MakeFloat3}.Result) {materialFunction=/Game/Athena/UI/Materials/MF_HSVtoRGB}.RGB_Out) {constB=0}
  n5 = ScalarParameter() {defaultValue=16, parameterName=NoColumns}
  n6 = Multiply(A=TextureCoordinate() {coordinateIndex=1}, B=@n5)
  n7 = Floor(None=@n6)
  MaterialExpressionComponentMask_1 = ComponentMask(None=@n7) {r=true}
  n8 = Multiply(A=@MaterialExpressionComponentMask_1) {constB=1.4}
  n9 = Multiply(A=@n8) {constB=0.064}
  n10 = Time()
  n11 = Multiply(A=@n10, B=ScalarParameter() {defaultValue=0.6667, parameterName=FequencyBarTimeScale})
  n12 = VertexColor()
  n13 = Multiply(A=Saturate(None=@n12.G)) {constB=5}
  n14 = Add(A=@n11, B=@n13)
  n15 = Multiply(A=@n14) {constB=0.6666}
  n16 = ConstantBiasScale(None=Sine(None=Add(A=@n9, B=@n15)))
  n17 = Multiply(A=@n8) {constB=0.9}
  n18 = Multiply(A=@n14)
  n19 = ConstantBiasScale(None=Sine(None=Add(A=@n17, B=@n18)))
  n20 = Power(Base=Add(A=@n16, B=@n19)) {constExponent=0.5, desc=higher more often}
  n21 = Multiply(A=Add(A=@MaterialExpressionComponentMask_1) {constB=17}) {constB=0.064}
  n22 = Multiply(A=ConstantBiasScale(None=Sine(None=@n21))) {constB=0.2}
  Add(A=@n20, B=@n22)
  Add(A=@MaterialExpressionComponentMask_1, B=Ceil(None=Multiply(A=@n10) {constB=3}))
  Constant3Vector() {constant=<rgba 0 0 1 0>}
  ConstantBiasScale(None=Sine(None=@n9))
  Constant()
  n23 = Saturate(None=Multiply(A=@MaterialExpressionComponentMask_1) {constB=0.05})
  n24 = MaterialFunctionCall!(useSmoothStep, uvx=@n23) {materialFunction=/Game/Accessories/FORT_Backpacks/Backpack_F_MED_Cactus_Rocker/Materials/SampleAudioMPCHistory}
  n25 = Saturate(None=@n24.Interpolated Spectrum)
  MaterialExpressionMultiply_31 = Multiply(A=ComponentMask(None=@n25) {r=true}) {constB=0.333, desc=Make small prong not go as high}
  MaterialExpressionComponentMask_5 = ComponentMask(None=@n25) {g=true}
  MaterialExpressionComponentMask_6 = ComponentMask(None=@n25) {b=true}
  MaterialExpressionComponentMask_7 = ComponentMask(None=@n25) {a=true}
  MaterialExpressionSaturate_1 = Saturate(None=@n24.Interpolated Amplitude Music)
  MaterialExpressionAdd_3 = Add(A=@n13) {constB=0.005}
  n26 = Custom!(
   in1=@MaterialExpressionMultiply_31
   in2=@MaterialExpressionComponentMask_5
   in3=@MaterialExpressionComponentMask_6
   in4=@MaterialExpressionComponentMask_7
   in5=@MaterialExpressionSaturate_1
   alpha=@MaterialExpressionAdd_3
  ) {
   code="##pragma once\r\n\r\nconst int numVals = 5;\r\n\r\n const float values[numVals] = {\r\n\t\tin1, \r\n        in2,\r\n\t\tin3, \r\n        in4,\r\n        in5\r\n\t};\r\nconst uint a = uint(floor(alpha));\r\n\r\n\r\nreturn values[a];"
   outputType=CMOT_Float1
   inputs [
    {
     inputName=in1
     input {expression=&self:MaterialExpressionMultiply_31}
    }
    {
     inputName=in2
     input {expression=&self:MaterialExpressionComponentMask_5}
    }
    {
     inputName=in3
     input {expression=&self:MaterialExpressionComponentMask_6}
    }
    {
     inputName=in4
     input {expression=&self:MaterialExpressionComponentMask_7}
    }
    {
     inputName=in5
     input {expression=&self:MaterialExpressionSaturate_1}
    }
    {
     inputName=alpha
     input {expression=&self:MaterialExpressionAdd_3}
    }
   ]
  }
  n27 = Add(A=ChannelMaskParameter(None=Divide(A=@n7, B=@n5)) {maskChannel=Green, parameterName=Levels, defaultValue=<rgba 0 1 0 0>}.R) {constB=0.08}
  n28 = Power(Base=Saturate(None=Multiply(A=@n20) {constB=0.8}))
  n29 = ScalarParameter() {defaultValue=0.4, parameterName=FrequencyAmbientHeight}
  n30 = Multiply(A=@n24.Interpolated Amplitude Music) {constB=0.9}
  MaterialExpressionAdd_4 = Add(
   A=Subtract(A=Power(Base=@n26) {constExponent=1.3333}, B=OneMinus(None=@n27))
   B=Multiply(A=Multiply(A=@n28, B=@n29) {constB=0.6666}, B=OneMinus(None=@n30))
  )
  Custom!(val=@MaterialExpressionAdd_4) {
   code="float tmp = 0;\r\nif(val > 0){\r\n tmp = 1;\r\n}\r\nreturn tmp;"
   outputType=CMOT_Float1
   inputs [{
     inputName=val
     input {expression=&self:MaterialExpressionAdd_4}
    }]
  }
  n31 = Custom!(co=@MaterialExpressionComponentMask_1) {
   code="    return frac(sin(dot(co, float2(12.9898, 78.233))) * 43758.5453);\r\n"
   description=rNd
   outputType=CMOT_Float1
   inputs [{
     inputName=co
     input {expression=&self:MaterialExpressionComponentMask_1}
    }]
  }
  MaterialExpressionAdd_10 = Add(A=@MaterialExpressionComponentMask_1)
  n32 = Custom(co=@MaterialExpressionAdd_10) {
   code="    return frac(sin(dot(co, float2(12.9898, 78.233))) * 43758.5453);\r\n"
   description=rNd
   outputType=CMOT_Float1
   inputs [{
     inputName=co
     input {expression=&self:MaterialExpressionAdd_10}
    }]
  }
  Divide(A=Add(A=@n31, B=@n32))
  Frac(None=@n10)
  Frac(None=@n11)
  n33 = FunctionInput() {inputName=MA, inputType=FunctionInput_MaterialAttributes}
  n34 = Constant3Vector() {constant=<rgba 1 0 1 0>}
  n35 = Power(Base=@n12.R) {constExponent=0.6667}
  MaterialExpressionLinearInterpolate_1 = LinearInterpolate(A=@n34, B=@n34, Alpha=@n35)
  n36 = Constant3Vector() {constant=<rgba 0 1 0.1667 0>}
  MaterialExpressionLinearInterpolate_2 = LinearInterpolate(A=@n34, B=@n36, Alpha=@n35)
  n37 = Constant3Vector() {constant=<rgba 0 0.5219 1 0>}
  n38 = Constant3Vector() {constant=<rgba 0.0834 0 1 0>}
  MaterialExpressionLinearInterpolate_3 = LinearInterpolate(A=@n37, B=@n38, Alpha=@n35)
  MaterialExpressionLinearInterpolate_4 = LinearInterpolate(A=@n34, B=@n36, Alpha=@n35)
  MaterialExpressionLinearInterpolate_5 = LinearInterpolate(A=@n36, B=@n34, Alpha=@n35)
  MaterialExpressionFloor_4 = Floor(None=@MaterialExpressionAdd_3)
  n39 = Custom!(
   in1=@MaterialExpressionLinearInterpolate_1
   in2=@MaterialExpressionLinearInterpolate_2
   in3=@MaterialExpressionLinearInterpolate_3
   in4=@MaterialExpressionLinearInterpolate_4
   in5=@MaterialExpressionLinearInterpolate_5
   alpha=@MaterialExpressionFloor_4
  ) {
   code="##pragma once\r\n\r\nconst int numVals = 5;\r\n\r\nconst float3 values[numVals] = {\r\n\t\tin1, \r\n        in2,\r\n\t\tin3, \r\n        in4,\r\n        in5\r\n\t};\r\n\r\nconst uint a = uint(alpha);\r\n\r\nreturn values[a];"
   inputs [
    {
     inputName=in1
     input {expression=&self:MaterialExpressionLinearInterpolate_1}
    }
    {
     inputName=in2
     input {expression=&self:MaterialExpressionLinearInterpolate_2}
    }
    {
     inputName=in3
     input {expression=&self:MaterialExpressionLinearInterpolate_3}
    }
    {
     inputName=in4
     input {expression=&self:MaterialExpressionLinearInterpolate_4}
    }
    {
     inputName=in5
     input {expression=&self:MaterialExpressionLinearInterpolate_5}
    }
    {
     inputName=alpha
     input {expression=&self:MaterialExpressionFloor_4}
    }
   ]
  }
  n40 = NamedRerouteUsage()
  n41 = Multiply(A=Multiply(A=@n39, B=Saturate(None=@n12.B)), B=Multiply(A=Sign(None=@MaterialExpressionAdd_4), B=Multiply(A=@n40)))
  n42 = BreakMaterialAttributes!(Attr=@n33)
  n43 = Multiply(A=Desaturation(None=@n42.EmissiveColor, Fraction=Constant() {r=1})) {constB=56}
  n44 = Multiply(A=Multiply(A=Saturate(None=@n41)) {constB=0.3333}, B=OneMinus(None=Saturate(None=@n43)))
  n45 = Saturate(None=NamedRerouteUsage())
  n46 = Constant() {r=4}
  n47 = Constant() {r=0.2}
  n48 = Power(Base=InverseLinearInterpolate(Value=NamedRerouteUsage()) {constA=0.45, constB=0.9})
  n49 = ScalarParameter() {defaultValue=1.75, parameterName=reactiveEmissiveGain}
  n50 = Multiply(
   A=Power(Base=OneMinus(None=@n45), Exp=LinearInterpolate(A=@n46, B=@n47, Alpha=@n48)) {constExponent=8}
   B=Add(A=ScalarParameter() {defaultValue=1.333, parameterName=baseEmissiveGain}, B=Multiply(A=@n48, B=@n49))
  )
  n51 = Add(A=@n44, B=Multiply(A=@n44, B=@n50) {constB=3})
  n52 = Multiply(A=@n42.EmissiveColor, B=Multiply(A=@n50) {constB=4, desc=grid specific boost to balance})
  n53 = MaterialFunctionCall!(Material=@n33, Emissive=Clamp(None=Add(A=@n51, B=@n52) {desc=add back original grid boost}, Min=Constant(), Max=Constant() {r=3.2})) {materialFunction=/Engine/Functions/MaterialLayerFunctions/MatLayerBlend_Emissive}
  n54 = MaterialFunctionCall!(Material=@n53.Blended Material) {materialFunction=/Engine/Functions/MaterialLayerFunctions/MatLayerBlend_BreakBaseColor}
  FunctionOutput(None=MaterialFunctionCall!(Material=@n53.Blended Material, NewBaseColor=Add(A=@n54.BaseColor, B=@n44), Mask) {materialFunction=/Engine/Functions/MaterialLayerFunctions/MatLayerBlend_OverrideBaseColor}.Blended Material) {outputName=MA}
  MaterialFunctionCall!(a=@n36, b=@n34, alpha=@n35) {materialFunction=&/ColorLerpNonWashedOut}
  MaterialFunctionCall!(useSmoothStep, uvx=@n23) {materialFunction=/Game/Accessories/FORT_Backpacks/Backpack_F_MED_Cactus_Rocker/Materials/SampleAudioMPCInterpolateed}
  MaterialFunctionCall(a=@n34, b=@n34, alpha=@n35) {materialFunction=&ColorLerpNonWashedOut}
  MaterialFunctionCall(a=@n34, b=@n36, alpha=@n35) {materialFunction=&ColorLerpNonWashedOut}
  MaterialFunctionCall(a=@n37, b=@n38, alpha=@n35) {materialFunction=&ColorLerpNonWashedOut}
  MaterialFunctionCall(a=@n34, b=@n36, alpha=@n35) {materialFunction=&ColorLerpNonWashedOut}
  Multiply(A=@n42.EmissiveColor) {constB=1.5}
  Multiply(A=@n16) {constB=0.5}
  Multiply(A=@n19) {constB=0.5}
  NamedRerouteDeclaration(None=Frac(None=@n6)) {name=CellUVCoords, nodeColor=<rgba 1 0.5391 0 1>}
  n55 = NamedRerouteUsage()
  n56 = Abs(None=Subtract(A=Multiply(A=@n55) {constB=2}))
  n57 = ChannelMaskParameter(None=@n56)
  n58 = Constant()
  n59 = ScalarParameter() {defaultValue=0.98, parameterName=CellWidth}
  n60 = ChannelMaskParameter(None=@n56) {maskChannel=Green, parameterName=Param_1, defaultValue=<rgba 0 1 0 0>}
  n61 = Constant()
  n62 = ScalarParameter() {defaultValue=0.98, parameterName=CellHeight}
  NamedRerouteDeclaration(None=Multiply(A=SphereMask(A=@n57.R, B=@n58, Radius=@n59) {attenuationRadius=0.4}, B=SphereMask(A=@n60.R, B=@n61, Radius=@n62) {attenuationRadius=0.4})) {name=CellsMask, nodeColor=<rgba 1 0 0.3281 1>}
  NamedRerouteDeclaration(None=@n24.Interpolated Amplitude Music) {name=amp, nodeColor=<rgba 0 0.2031 1 1>}
  NamedRerouteDeclaration(None=@n12.R) {name=ramp, nodeColor=<rgba 0.5579 0 1 1>}
  Saturate(None=@MaterialExpressionAdd_4)
  ScalarParameter() {parameterName=ghue}
  Subtract(B=@n22)
 ]}
